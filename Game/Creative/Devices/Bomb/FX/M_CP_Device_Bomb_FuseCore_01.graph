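# Translucent unlit Niagara sprite material for the fuse core of the Creative bomb device.
# Per-particle random drives a sine twinkle and hue shift on particle color, blended against the plain color by a dynamic parameter before emissive adjustments.
# A flare texture, sharpened by a dynamic power exponent and scaled by particle alpha, forms the mask.
# A dynamic parameter also drives camera offset; the similarly named scalar parameters are unconnected.
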Material /Game/Creative/Devices/Bomb/FX/M_CP_Device_Bomb_FuseCore_01
graph {
 properties {
  bEnableResponsiveAA=true
  bUsedWithNiagaraSprites=true
  blendMode=BLEND_Translucent
  pixelDepthOffsetMode=PDOM_Legacy
  shadingModel=MSM_Unlit
  usedShadingModels=Unlit
 }
 nodes [
  n1 = ParticleColor()
  n2 = ParticleRandom()
  n3 = DynamicParameter() {paramNames [Use Random HueShift Color, Hueshift Range, TwinkleSpeed, Camera Offset]}
  n4 = Multiply(A=@n2, B=@n3.Hueshift Range) {constB=0.25}
  n5 = MaterialFunctionCall(Speed=Multiply(A=@n2, B=@n3.TwinkleSpeed)) {materialFunction=&DebugTimeSine}
  n6 = Multiply(A=Clamp(None=Multiply(A=@n5.Result) {constB=5}) {maxDefault=5, minDefault=0.25}, B=@n1.RGB)
  MaterialFunctionCall(Emissive=LinearInterpolate(
    A=@n1.RGB
    B=MaterialFunctionCall(Hue Shift Percentage=@n4, Texture=@n6) {materialFunction=&HueShift}.Result
    Alpha=@n3.Use Random HueShift Color
   )) {materialFunction=&MF_FX_EmissiveAdjustments}
  MaterialFunctionCall(Offset Amount=@n3.Camera Offset) {materialFunction=/Engine/Functions/Engine_MaterialFunctions02/WorldPositionOffset/CameraOffset}
  n7 = TextureSampleParameter2D() {parameterName=Texture, texture=/Game/Effects/Fort_Effects/Textures/T_Emote_Meticulous_Flare_03}
  n8 = DynamicParameter() {parameterIndex=1, paramNames [Mask Power Exponent, Param2, Param3, Param4]}
  Reroute(None=Multiply(A=@n1.A, B=Power(Base=@n7.R, Exp=@n8.Mask Power Exponent)))
  ScalarParameter() {defaultValue=1, parameterName=Use Particle Random Color}
  ScalarParameter() {defaultValue=25, parameterName="Offset Amount (S)"}
  ScalarParameter() {defaultValue=0.25, parameterName=HueShift Range}
  ScalarParameter() {defaultValue=3, parameterName=TwinkleSpeed}
 ]
}
